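# Prototype Landmass erosion material, unlit, that computes an advection step over a height render target holding terrain height, water height, water volume and sediment fraction.
# Advected terrain height is floored at local height minus a MaxAdvectSlope drop, then passed through SlopeSmoothing.
# Water volume is backtraced along neighbor-averaged VelocityRT, gains rain, and is packed with height into a float3.
# Solubility, sediment-dilution and box-mask branches feed nothing further in the graph shown.
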Material /Landmass/Landscape/OldPrototype_BP/Mats/Advect/M_Advect_Height
graph {
 properties {
  bAllowNegativeEmissiveColor=true
  blendMode=BLEND_AlphaComposite
  pixelDepthOffsetMode=PDOM_Legacy
  shadingModel=MSM_Unlit
  usedShadingModels=Unlit
 }
 nodes [
  n1 = MaterialFunctionCall() {materialFunction=/Landmass/Landscape/OldPrototype_BP/Mats/MF/PointSampleUVs}
  n2 = Divide(B=AppendVector(
    A=CollectionParameter() {collection=/Landmass/Landscape/BlueprintBrushes/MPC/MPC_Erosion, parameterName=ResX}
    B=CollectionParameter() {collection=&MPC_Erosion, parameterName=ResY}
   ))
  n3 = Multiply(A=@n2, B=Constant2Vector() {r=1})
  n4 = Add(
   A=TextureSampleParameter2D(UVs=@n1.TexelCenterUVs) {parameterName=VelocityRT, texture=&Black}.R
   B=TextureSampleParameter2D(UVs=Add(A=@n1.TexelCenterUVs, B=@n3)) {parameterName=VelocityRT, texture=&Black}.R
  )
  n5 = Multiply(A=@n2, B=Constant2Vector() {g=1})
  n6 = Add(
   A=TextureSampleParameter2D(UVs=@n1.TexelCenterUVs) {parameterName=VelocityRT, texture=&Black}.G
   B=TextureSampleParameter2D(UVs=Add(A=@n1.TexelCenterUVs, B=@n5)) {parameterName=VelocityRT, texture=&Black}.G
  )
  n7 = Multiply(
   A=CollectionParameter() {collection=&MPC_Erosion, parameterName=deltaT}
   B=AppendVector(A=Divide(A=@n4), B=Divide(A=@n6))
  )
  MaterialExpressionMultiply_31 = Multiply(
   A=Divide(
    A=MaterialFunctionCall(Vector 2=@n7) {materialFunction=/Engine/Functions/Engine_MaterialFunctions02/Utility/VectorLength}.V2 Length
    B=CollectionParameter() {collection=&MPC_Erosion, parameterName=GridSize}
   )
   B=CollectionParameter() {collection=&MPC_Erosion, parameterName=Solubility}
  )
  Custom!(t=@MaterialExpressionMultiply_31) {
   code="return 1 - exp(-t);"
   desc=Solubility can only approach 1
   description=exp
   outputType=CMOT_Float1
   inputs [{inputName=t, input {expression=&self:MaterialExpressionMultiply_31}}]
  }
  n8 = MaterialFunctionCall() {materialFunction=/Landmass/Landscape/OldPrototype_BP/Mats/MF/VelocityAdvect}
  n9 = TextureSampleParameter2D!(UVs=@n8.Advected UVs, Level=Constant(), Apply View MipBias) {mipValueMode=TMVM_MipLevel, parameterName=HeightRT, texture=&Black}
  n10 = Subtract(
   A=TextureSampleParameter2D(UVs=MaterialFunctionCall() {materialFunction=&PointSampleUVs}.TexelCenterUVs, Level=Constant()) {mipValueMode=TMVM_MipLevel, parameterName=HeightRT, texture=&Black}.R
   B=Multiply(
    A=MaterialFunctionCall(Vector 2=@n8.OffsetWorldSpace) {materialFunction=&VectorLength}.V2 Length
    B=CollectionParameter() {collection=&MPC_Erosion, parameterName=MaxAdvectSlope}
   ) {constB=0.1}
  )
  n11 = Multiply(
   A=AppendVector(
    A=CollectionParameter() {collection=&MPC_Erosion, parameterName=ResX}
    B=CollectionParameter() {collection=&MPC_Erosion, parameterName=ResY}
   )
   B=CollectionParameter() {collection=&MPC_Erosion, parameterName=GridSize}
  )
  n12 = TextureSampleParameter2D(
   UVs=Subtract(
    A=MaterialFunctionCall() {materialFunction=&PointSampleUVs}.TexelCenterUVs
    B=Divide(A=@n7, B=@n11)
   )
   Level=Constant()
  ) {mipValueMode=TMVM_MipLevel, parameterName=HeightRT, texture=&Black}
  n13 = Add(
   A=@n12.B
   B=Multiply(
    A=CollectionParameter() {collection=&MPC_Erosion, parameterName=RainAmount}
    B=CollectionParameter() {collection=&MPC_Erosion, parameterName=deltaT}
   )
  )
  MaterialFunctionCall(
   X=MaterialFunctionCall!(UVs=@n8.Advected UVs, Height=Max(A=@n9.R, B=@n10)) {materialFunction=/Landmass/Landscape/OldPrototype_BP/Mats/MF/SlopeSmoothing}.Height
   Z=@n13
  ) {materialFunction=&MakeFloat3}
  MaterialFunctionCall() {materialFunction=&PointSampleUVs}
  n14 = TextureCoordinate()
  n15 = Constant() {r=0.5}
  n16 = ScalarParameter() {defaultValue=0.95}
  n17 = ScalarParameter() {defaultValue=0.02}
  Multiply(B=Multiply(A=MaterialFunctionCall!(A=@n14, B=@n15, Bounds=@n16, Edge Falloff=@n17) {materialFunction=/Engine/Functions/Engine_MaterialFunctions01/Gradient/BoxMask-2D}.Result))
  n18 = Max(A=@n13) {constB=0}
  OneMinus(None=Multiply(A=Divide(A=@n12.B, B=@n18) {desc=Dilution Fraction}, B=@n12.A) {desc=Current Sediment Fraction Diluted by new Water})
  n19 = CollectionParameter() {collection=&MPC_Erosion, parameterName=RainAmount}
  n20 = CollectionParameter() {collection=&MPC_Erosion, parameterName=deltaT}
  Subtract(
   A=TextureSampleParameter2D!(UVs=MaterialFunctionCall() {materialFunction=&VelocityAdvect}.Advected UVs, Apply View MipBias) {desc="R=TerrainZ, G=WaterZ, B=Water Volume, A=Sediment Fraction", parameterName=HeightRT, texture=&Black}.R
   B=Multiply(
    A=Multiply(A=@n19, B=@n20)
    B=CollectionParameter() {collection=&MPC_Erosion, parameterName=Solubility}
   )
  )
  TextureCoordinate()
  TextureCoordinate()
 ]
}
