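# Unlit prototype Landmass material that computes semi-Lagrangian self-advection of a 2D velocity field stored in VelocityRT.
# Each component gets its own backtrace vector, pairing itself with the other component sampled at a diagonal half-texel offset, as on a staggered grid.
# Vectors scale by deltaT, resolution and GridSize from MPC_Erosion, then offset point-sampled UVs to fetch new R and G for MakeFloat3.
# An older neighbor-averaging branch remains unused.
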Material /Landmass/Landscape/OldPrototype_BP/Mats/Advect/M_Advect_Velocity
graph {
 properties {bAllowNegativeEmissiveColor=true, pixelDepthOffsetMode=PDOM_Legacy, shadingModel=MSM_Unlit, usedShadingModels=Unlit}
 nodes [
  n1 = TextureCoordinate()
  n2 = TextureSampleParameter2D(UVs=@n1) {parameterName=VelocityRT, texture=&Black}
  n3 = Divide(B=AppendVector(A=ScalarParameter() {defaultValue=1024, parameterName=SizeX}, B=ScalarParameter() {defaultValue=1024, parameterName=SizeY}))
  n4 = Constant2Vector() {r=1}
  n5 = TextureSampleParameter2D(UVs=Subtract(A=@n1, B=Multiply(A=@n3, B=@n4))) {parameterName=VelocityRT, texture=&Black}
  n6 = Constant2Vector() {g=1}
  n7 = TextureSampleParameter2D(UVs=Subtract(A=@n1, B=Multiply(A=@n3, B=@n6))) {parameterName=VelocityRT, texture=&Black}
  AppendVector(A=Divide(A=Add(A=@n2.R, B=@n5.R)), B=Divide(A=Add(A=@n2.G, B=@n7.G)))
  n8 = MaterialFunctionCall() {materialFunction=/Landmass/Landscape/OldPrototype_BP/Mats/MF/PointSampleUVs}
  n9 = TextureSampleParameter2D!(
   UVs=MaterialFunctionCall() {materialFunction=&PointSampleUVs}.TexelCenterUVs
   Level=Constant()
   Apply View MipBias
  ) {constMipValue=0, mipValueMode=TMVM_MipLevel, parameterName=VelocityRT, texture=&Black}
  n10 = CollectionParameter() {collection=/Landmass/Landscape/BlueprintBrushes/MPC/MPC_Erosion, parameterName=ResX}
  n11 = CollectionParameter() {collection=&MPC_Erosion, parameterName=ResY}
  n12 = Multiply(A=Divide(B=AppendVector(A=@n10, B=@n11))) {constB=0.5}
  n13 = Constant2Vector() {g=1, r=-1}
  n14 = TextureSampleParameter2D(
   UVs=Add(
    A=MaterialFunctionCall() {materialFunction=&PointSampleUVs}.TexelCenterUVs
    B=Multiply(A=@n12, B=@n13)
   )
   Level=Constant()
  ) {constMipValue=0, mipValueMode=TMVM_MipLevel, parameterName=VelocityRT, texture=&Black}
  n15 = CollectionParameter() {collection=&MPC_Erosion, parameterName=deltaT}
  n16 = Multiply(
   A=AppendVector(
    A=CollectionParameter() {collection=&MPC_Erosion, parameterName=ResX}
    B=CollectionParameter() {collection=&MPC_Erosion, parameterName=ResY}
   )
   B=CollectionParameter() {collection=&MPC_Erosion, parameterName=GridSize}
  )
  n17 = Divide(A=Multiply(A=AppendVector(A=@n9.R, B=@n14.G), B=@n15), B=@n16)
  n18 = MaterialFunctionCall() {materialFunction=&PointSampleUVs}
  n19 = Constant2Vector() {g=-1, r=1}
  n20 = TextureSampleParameter2D(
   UVs=Add(
    A=MaterialFunctionCall() {materialFunction=&PointSampleUVs}.TexelCenterUVs
    B=Multiply(A=@n12, B=@n19)
   )
   Level=Constant()
  ) {constMipValue=0, mipValueMode=TMVM_MipLevel, parameterName=VelocityRT, texture=&Black}
  n21 = Divide(A=Multiply(A=@n15, B=AppendVector(A=@n20.R, B=@n9.G)), B=@n16)
  MaterialFunctionCall(
   X=TextureSampleParameter2D(UVs=Subtract(A=@n8.TexelCenterUVs, B=@n17), Level=Constant()) {desc=New, mipValueMode=TMVM_MipLevel, parameterName=VelocityRT, texture=&Black}.R
   Y=TextureSampleParameter2D(UVs=Subtract(A=@n18.TexelCenterUVs, B=@n21), Level=Constant()) {desc=New, mipValueMode=TMVM_MipLevel, parameterName=VelocityRT, texture=&Black}.G
  ) {materialFunction=&MakeFloat3}
  TextureCoordinate()
  TextureCoordinate()
  TextureCoordinate()
  TextureCoordinate()
  TextureSampleParameter2D!(UVs, Apply View MipBias) {desc=Old, parameterName=VelocityRT, texture=&Black}
 ]
}
