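# Unlit render-target material for a prototype landscape erosion simulation, updating height, water and sediment per texel.
# VelocityRT wall differences give divergence, which updates HeightRT water before evaporation rescales the sediment fraction.
# Cell speed sets an exponential solubility limit; comparing sediment against it selects dissolution or deposition, adjusting height and sediment.
# AlphaComposite with negative emissive suits signed data output; a custom-HLSL slope term is computed but feeds nothing shown.
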Material /Landmass/Landscape/OldPrototype_BP/Mats/Advect/M_CalcDivergence
graph {
 properties {
  bAllowNegativeEmissiveColor=true
  blendMode=BLEND_AlphaComposite
  pixelDepthOffsetMode=PDOM_Legacy
  shadingModel=MSM_Unlit
  usedShadingModels=Unlit
 }
 nodes [
  n1 = TextureSampleParameter2D!(
   UVs=MaterialFunctionCall() {materialFunction=/Landmass/Landscape/OldPrototype_BP/Mats/MF/PointSampleUVs}.TexelCenterUVs
   Level=Constant()
   Apply View MipBias
  ) {constMipValue=0, mipValueMode=TMVM_MipLevel, parameterName=HeightRT, texture=&Black}
  n2 = MaterialFunctionCall() {materialFunction=&PointSampleUVs}
  n3 = CollectionParameter() {collection=/Landmass/Landscape/BlueprintBrushes/MPC/MPC_Erosion, parameterName=ResX}
  n4 = CollectionParameter() {collection=&MPC_Erosion, parameterName=ResY}
  n5 = Multiply(A=Divide(B=AppendVector(A=@n3, B=@n4)))
  n6 = Constant2Vector() {r=1}
  n7 = TextureSampleParameter2D(UVs=Add(A=@n2.TexelCenterUVs, B=Multiply(A=@n5, B=@n6))) {constMipValue=0, desc=Right Wall, mipValueMode=TMVM_MipLevel, parameterName=VelocityRT, texture=&Black}
  n8 = TextureSampleParameter2D(UVs=@n2.TexelCenterUVs) {constMipValue=0, desc=Left Wall, mipValueMode=TMVM_MipLevel, parameterName=VelocityRT, texture=&Black}
  n9 = Subtract(A=@n7.R, B=@n8.R)
  n10 = MaterialFunctionCall() {materialFunction=&PointSampleUVs}
  n11 = Constant2Vector() {g=1}
  n12 = TextureSampleParameter2D(UVs=Add(A=@n10.TexelCenterUVs, B=Multiply(A=@n5, B=@n11))) {constMipValue=0, desc=Top Wall, mipValueMode=TMVM_MipLevel, parameterName=VelocityRT, texture=&Black}
  n13 = TextureSampleParameter2D(UVs=@n10.TexelCenterUVs) {constMipValue=0, desc=Bottom Wall, mipValueMode=TMVM_MipLevel, parameterName=VelocityRT, texture=&Black}
  n14 = Subtract(A=@n12.G, B=@n13.G)
  n15 = CollectionParameter() {collection=&MPC_Erosion, parameterName=deltaT}
  n16 = Multiply(
   A=Divide(A=Add(A=@n9, B=@n14), B=CollectionParameter() {collection=&MPC_Erosion, parameterName=GridSize})
   B=@n15
  )
  n17 = Max(A=Divide(A=@n1.B, B=Add(B=@n16) {constA=1})) {constB=0, desc=Updated Water}
  n18 = Subtract(
   A=@n17
   B=Multiply(
    A=CollectionParameter() {collection=&MPC_Erosion, parameterName=Evaporation}
    B=CollectionParameter() {collection=&MPC_Erosion, parameterName=deltaT}
   )
  )
  n19 = Max(A=@n18) {constB=0}
  n20 = Multiply(A=Divide(A=@n17, B=Max(A=@n18) {constB=0.0001}) {desc=Solubility change}, B=@n1.A) {desc=Updated Sediment Fraction}
  n21 = AppendVector(A=Add(A=@n8.R, B=@n7.R), B=Add(A=@n13.G, B=@n12.G))
  n22 = MaterialFunctionCall(Vector 2=Multiply(A=@n15, B=Divide(A=@n21) {desc=Cell Velocity})) {materialFunction=/Engine/Functions/Engine_MaterialFunctions02/Utility/VectorLength}
  MaterialExpressionMultiply_31 = Multiply(
   A=Divide(A=Add(A=@n22.V2 Length), B=CollectionParameter() {collection=&MPC_Erosion, parameterName=GridSize})
   B=CollectionParameter() {collection=&MPC_Erosion, parameterName=Solubility}
  )
  n23 = Custom!(t=@MaterialExpressionMultiply_31) {
   code="return 1 - exp(-t);"
   desc=Solubility can only approach 1
   description=exp
   outputType=CMOT_Float1
   inputs [{inputName=t, input {expression=&self:MaterialExpressionMultiply_31}}]
  }
  n24 = Subtract(A=@n20, B=Add(A=@n23, B=ScalarParameter() {parameterName=asdfdf})) {desc="Position = Over saturated"}
  n25 = ScalarParameter() {defaultValue=0.05, parameterName=Per Step Dissolution}
  n26 = ScalarParameter() {defaultValue=0.001, parameterName=Per Step Deposition}
  n27 = Saturate(None=Ceil(None=@n24))
  n28 = Multiply(A=@n24, B=Multiply(A=LinearInterpolate(A=@n25, B=@n26, Alpha=@n27), B=ScalarParameter() {parameterName=Enable Erosion})) {desc=Over-under amount}
  n29 = Multiply(A=@n19, B=@n28)
  Clamp(None=MaterialFunctionCall(X=Add(A=@n1.R, B=@n29), Y=@n1.G, Z=@n19) {materialFunction=&MakeFloat3}.Result) {maxDefault=32767, minDefault=-32768}
  Min(A=@n24) {constB=0}
  MaterialExpressionTextureObjectParameter_0 = TextureObjectParameter() {parameterName=HeightRT, texture=&Black}
  MaterialExpressionDivide_6 = Divide(B=AppendVector(
    A=CollectionParameter() {collection=&MPC_Erosion, parameterName=ResX}
    B=CollectionParameter() {collection=&MPC_Erosion, parameterName=ResY}
   ))
  MaterialExpressionCollectionParameter_5 = CollectionParameter() {collection=&MPC_Erosion, parameterName=GridSize}
  MaterialExpressionTextureCoordinate_1 = TextureCoordinate()
  n30 = Custom!(
   HeightRT=@MaterialExpressionTextureObjectParameter_0
   TexelSize=@MaterialExpressionDivide_6
   GridSize=@MaterialExpressionCollectionParameter_5
   UV=@MaterialExpressionTextureCoordinate_1
  ) {
   code="//HeightRT: input Texture Object containing heights.R and water.B\r\n//TextureSize: input for size\r\n//GridSize: input float grid world size\r\n//HeightScale: input for height range scale\r\n\r\nfloat2 offsets[4] = {float2(0,1), float2(0,0), float2(1,0), float2(0,0)};\r\n\r\nfloat2 texelsize = TexelSize;\r\n\r\n\r\nfloat PosYSample = HeightRT.SampleLevel(HeightRTSampler, UV + offsets[0] * texelsize, 0).r;\r\nfloat NegYSample = HeightRT.SampleLevel(HeightRTSampler, UV + offsets[1] * texelsize, 0).r;\r\nfloat PosXSample = HeightRT.SampleLevel(HeightRTSampler, UV + offsets[2] * texelsize, 0).r;\r\nfloat NegXSample = HeightRT.SampleLevel(HeightRTSampler, UV + offsets[3] * texelsize, 0).r;\r\n\r\nfloat3 tangent =  float3( 2*GridSize, 0, ((PosXSample) - (NegXSample)));\r\nfloat3 binormal = float3( 0, 2*GridSize, ((PosYSample) - (NegYSample)));\r\ntangent = normalize(tangent);\r\nbinormal = normalize(binormal);\r\n\r\nfloat3 normal = normalize(cross(tangent, binormal));\r\n\r\nreturn normal;"
   description=Normal
   inputs [
    {
     inputName=HeightRT
     input {expression=&self:MaterialExpressionTextureObjectParameter_0}
    }
    {inputName=TexelSize, input {expression=&self:MaterialExpressionDivide_6}}
    {
     inputName=GridSize
     input {expression=&self:MaterialExpressionCollectionParameter_5}
    }
    {inputName=UV, input {expression=&self:MaterialExpressionTextureCoordinate_1}}
   ]
  }
  n31 = ComponentMask(None=@n30) {b=true, desc=cosine}
  n32 = Multiply(A=@n31, B=@n31)
  Multiply(A=@n23, B=SquareRoot(None=OneMinus(None=@n32)) {desc=sine})
  OneMinus(None=Subtract(A=@n20, B=@n28))
 ]
}
