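# Unlit simulation material in the old Landmass prototype that updates fluid velocity from gravity and wind, using MPC_Erosion parameters.
# A Custom HLSL node point-samples HeightRT, sums terrain height and fluid volume, and accelerates flow by surface differences to -X and -Y neighbors.
# Where fluid exists, it adds VelocityRT and time-varying noise-modulated wind, then applies exponential damping.
# Negative emissive is enabled, allowing signed velocity output.
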Material /Landmass/Landscape/OldPrototype_BP/Mats/Advect/M_GravAcceleration
graph {
 properties {bAllowNegativeEmissiveColor=true, pixelDepthOffsetMode=PDOM_Legacy, shadingModel=MSM_Unlit, usedShadingModels=Unlit}
 nodes [
  MaterialExpressionTextureObjectParameter_0 = TextureObjectParameter() {parameterName=HeightRT, texture=&Black}
  MaterialExpressionTextureSampleParameter2D_7 = TextureSampleParameter2D!(UVs, Level=Constant(), Apply View MipBias) {mipValueMode=TMVM_MipLevel, parameterName=VelocityRT, texture=&Black}
  MaterialExpressionCollectionParameter_9 = CollectionParameter() {collection=/Landmass/Landscape/BlueprintBrushes/MPC/MPC_Erosion, parameterName=deltaT}
  MaterialExpressionDivide_0 = Divide(B=AppendVector(
    A=CollectionParameter() {collection=&MPC_Erosion, parameterName=ResX}
    B=CollectionParameter() {collection=&MPC_Erosion, parameterName=ResY}
   ))
  MaterialExpressionCollectionParameter_8 = CollectionParameter() {collection=&MPC_Erosion, parameterName=GridSize}
  MaterialExpressionMaterialFunctionCall_0 = MaterialFunctionCall() {materialFunction=/Landmass/Landscape/OldPrototype_BP/Mats/MF/PointSampleUVs}
  MaterialExpressionCollectionParameter_0 = CollectionParameter() {collection=&MPC_Erosion, parameterName=VelocityDamping}
  n1 = TextureSample(UVs=MaterialFunctionCall(Speed=ScalarParameter() {defaultValue=0.1, parameterName=Wind Change Speed}) {materialFunction=&TimeWithSpeedVariable}.Frac Time) {samplerType=SAMPLERTYPE_LinearGrayscale, texture=/Engine/Functions/Engine_MaterialFunctions02/ExampleContent/Textures/LowResBlurredNoise}
  n2 = Subtract(A=@n1.R) {constB=0.6}
  MaterialExpressionMultiply_0 = Multiply(A=VectorParameter() {parameterName=V}.RGB, B=@n2)
  Custom!(
   HeightRT=@MaterialExpressionTextureObjectParameter_0
   Velocity=@MaterialExpressionTextureSampleParameter2D_7.RGB
   deltaT=@MaterialExpressionCollectionParameter_9
   TexelSize=@MaterialExpressionDivide_0
   GridSize=@MaterialExpressionCollectionParameter_8
   inUV=@MaterialExpressionMaterialFunctionCall_0.TexelCenterUVs
   Damping=@MaterialExpressionCollectionParameter_0
   Wind=@MaterialExpressionMultiply_0
  ) {
   code="//HeightData: input v4, R=Terrain Height, G=Fluid Height(bottom), B=FluidVolulume, A=Sediment fraction\r\n//Velocity: input \r\n//NormalRT: input \r\n//HeightRT: input\r\n//deltaT: input\r\n//input TextureSize\r\n//input TexelSize\r\n//input GridSize\r\n\r\nconst float grav = 980;\r\n\r\nfloat2 startZ = HeightRT.SampleLevel(HeightRTSampler, inUV, 0).rb;\r\nfloat2 offsetX = HeightRT.SampleLevel(HeightRTSampler, inUV - float2(TexelSize.x, 0), 0).rb;\r\nfloat2 offsetY = HeightRT.SampleLevel(HeightRTSampler, inUV - float2(0, TexelSize.y), 0).rb;\r\n\r\nstartZ.x += startZ.y;\r\noffsetX.x += offsetX.y;\r\noffsetY.x += offsetY.y;\r\n\r\nfloat3 outvel = 0;\r\n\r\nif(startZ.y > 0.01 || offsetX.y > 0.01 || offsetY.y > 0.01)\r\n{\r\n\r\n\t\toutvel.x = grav * ((offsetX.x-startZ.x) / GridSize) * deltaT;\r\n\t\toutvel.y = grav * ((offsetY.x-startZ.x) / GridSize) * deltaT;\r\n\t\t\r\n\t\tif( (startZ.y == 0 &&&& startZ.x > offsetX.x) || (offsetX.y == 0 &&&& offsetX.x> startZ.x))\r\n\t\t{\r\n\t\t//\toutvel.x = 0;\r\n\t\t}\r\n\t\t\r\n\t\tif( (startZ.y == 0 &&&& startZ.x > offsetY.x) || (offsetY.y == 0 &&&& offsetY.x> startZ.x) )\r\n\t\t{\r\n\t\t//\toutvel.y = 0;\r\n\t\t}\r\n\t\t\r\n\t\toutvel += Velocity;\r\n\t\toutvel += Wind * startZ.y * deltaT;\r\n\r\n}\r\n\r\nif(1)\r\n{\r\n\r\n\toutvel *= exp(-Damping*deltaT);\r\n    //velocity *= (1-Damping);\r\n}\r\n\r\nif(startZ.y == 0)\r\n{\r\n    //outvel = 0;\r\n}\r\n\r\nreturn outvel;\r\n\r\n"
   description=Velocity Acceleration
   inputs [
    {
     inputName=HeightRT
     input {expression=&self:MaterialExpressionTextureObjectParameter_0}
    }
    {
     inputName=Velocity
     input {
      expression=&self:MaterialExpressionTextureSampleParameter2D_7
      mask=1
      maskB=1
      maskG=1
      maskR=1
     }
    }
    {
     inputName=deltaT
     input {expression=&self:MaterialExpressionCollectionParameter_9}
    }
    {inputName=TexelSize, input {expression=&self:MaterialExpressionDivide_0}}
    {
     inputName=GridSize
     input {expression=&self:MaterialExpressionCollectionParameter_8}
    }
    {
     inputName=inUV
     input {expression=&self:MaterialExpressionMaterialFunctionCall_0}
    }
    {
     inputName=Damping
     input {expression=&self:MaterialExpressionCollectionParameter_0}
    }
    {inputName=Wind, input {expression=&self:MaterialExpressionMultiply_0}}
   ]
  }
  Multiply(A=@n2) {constB=-1}
  TextureCoordinate()
 ]
}
